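# LookAround is a humanoid AI action StateTree that, by its setup, handles bot focus and crouching.
# A global task gets the target actor, and a single Root state runs a Fortnite focus task alongside a bot crouch task.
# The focus task sets focus and reverts it on exit, while the crouch task requests a crouch.
# Root defines no transitions, and no bindings from the target actor to the focus task appear.
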StateTree /AICore/Humanoids/StateTrees/Actions/LookAround
global_tasks [{
  id=#1
  description=Get Target Actor
  properties {
   iD=#2
   instance {_structType=&FortStateTreeGetTargetActorTaskInstanceData}
   node {name=Get Target Actor, _structType=&FortStateTreeGetTargetActorTask}
  }
 }]
states [Root {
  id=#3
  type=STATE
  tasks [
   {
    id=#4
    description="(Fortnite) Focus a specific Actor and/or point/location"
    properties {
     iD=#5
     instance {
      _structType=/Script/FortniteAIServer.FortAthenaFocusAtStateTreeTaskInstanceData
      actor=null
      bSetBackOnExit=true
      bShouldSetFocus=true
      focusActor=null
      focusActorOffset=<0 0 0>
      focusWorldPoint=<0 0 0>
     }
     node {
      name="(Fortnite) Focus a specific Actor and/or point/location"
      _structType=/Script/FortniteAIServer.FortAthenaFocusAtStateTreeTask
      bCanEditConsideredForCompletion=true
      bConsideredForCompletion=true
      bHasTaskCompletionDelegateDispatcher=false
      bTaskEnabled=true
      transitionHandlingPriority=Normal
      bindingsBatch {value=65535}
      executionRuntimeTemplateIndex {value=65535}
      instanceDataHandle {index=65535, source=None, stateHandle {index=65535}}
      instanceTemplateIndex {value=65535}
      outputBindingsBatch {value=65535}
     }
    }
   }
   {
    id=#6
    description=Bot Crouch Task
    properties {
     iD=#7
     instance {_structType=/Script/FortniteAIServer.FortStateTreeBotCrouchTaskInstanceData, bShouldCrouch=true, botController=null}
     node {
      name=Bot Crouch Task
      _structType=/Script/FortniteAIServer.FortStateTreeBotCrouchTask
      bCanEditConsideredForCompletion=true
      bConsideredForCompletion=true
      bHasTaskCompletionDelegateDispatcher=false
      bTaskEnabled=true
      transitionHandlingPriority=Normal
      bindingsBatch {value=65535}
      executionRuntimeTemplateIndex {value=65535}
      instanceDataHandle {index=65535, source=None, stateHandle {index=65535}}
      instanceTemplateIndex {value=65535}
      outputBindingsBatch {value=65535}
     }
    }
   }
  ]
  transitions []
 }]
