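# Humanoid bot StateTree building block that equips, fires and reloads a weapon, with global tasks reading current weapon and magazine ammo.
# EquipItem runs for a valid object, entering Loop on success or failing the tree.
# Loop selects Reload or FireWeapon through bool and integer checks, falling back to Wait, which reselects every tick.
# FireWeapon combines pattern-based held fire with ADS control, returning to Loop when its integer equality check passes.
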StateTree /AICore/Humanoids/StateTrees/Actions/BuildingBlock/ManageWeapon
global_tasks [
 {
  id=#1
  description=Bot Get Current Weapon
  properties {
   iD=#2
   instance {_structType=/Script/FortniteAIServer.FortStateTreeBotGetCurrentWeaponTaskInstanceData, botController=null, currentWeapon=null}
   node {
    name=Bot Get Current Weapon
    _structType=/Script/FortniteAIServer.FortStateTreeBotGetCurrentWeaponTask
    bCanEditConsideredForCompletion=true
    bConsideredForCompletion=true
    bHasTaskCompletionDelegateDispatcher=false
    bTaskEnabled=true
    transitionHandlingPriority=Normal
    bindingsBatch {value=65535}
    executionRuntimeTemplateIndex {value=65535}
    instanceDataHandle {index=65535, source=None, stateHandle {index=65535}}
    instanceTemplateIndex {value=65535}
    outputBindingsBatch {value=65535}
   }
  }
 }
 {
  id=#3
  description=Get Weapon Magazine Ammo
  properties {
   iD=#4
   instance {_structType=/Script/FortniteAIServer.FortStateTreeGetWeaponMagazineAmmoTaskInstanceData, ammoInMagazine=0, weapon=null}
   node {
    name=Get Weapon Magazine Ammo
    _structType=/Script/FortniteAIServer.FortStateTreeGetWeaponMagazineAmmoTask
    bCanEditConsideredForCompletion=true
    bConsideredForCompletion=true
    bHasTaskCompletionDelegateDispatcher=false
    bTaskEnabled=true
    transitionHandlingPriority=Normal
    bindingsBatch {value=65535}
    executionRuntimeTemplateIndex {value=65535}
    instanceDataHandle {index=65535, source=None, stateHandle {index=65535}}
    instanceTemplateIndex {value=65535}
    outputBindingsBatch {value=65535}
   }
  }
 }
]
states [
 Root {
  id=#5
  type=STATE
  children [#6, #7]
  tasks []
  transitions []
 }
 EquipItem {
  id=#6
  type=STATE
  enter_conditions [{
    id=#8
    description=Is Object Valid
    properties {
     iD=#9
     instance {_structType=/Script/StateTreeModule.StateTreeObjectIsValidConditionInstanceData, object=null}
     node {
      name=None
      _structType=/Script/StateTreeModule.StateTreeObjectIsValidCondition
      bInvert=false
      deltaIndent=0
      evaluationMode=Evaluated
      operand=And
      bindingsBatch {value=65535}
      executionRuntimeTemplateIndex {value=65535}
      instanceDataHandle {index=65535, source=None, stateHandle {index=65535}}
      instanceTemplateIndex {value=65535}
      outputBindingsBatch {value=65535}
     }
    }
   }]
  tasks [{
    id=#10
    description=Equip Item
    properties {
     iD=#11
     instance {_structType=/Script/FortniteAIServer.FortStateTreeBotEquipItemTaskInstanceData, bValidateAbility=false, controller=null, itemToEquip=null}
     node {
      name=None
      _structType=/Script/FortniteAIServer.FortStateTreeBotEquipItemTask
      bCanEditConsideredForCompletion=true
      bConsideredForCompletion=true
      bHasTaskCompletionDelegateDispatcher=false
      bTaskEnabled=true
      transitionHandlingPriority=Normal
      bindingsBatch {value=65535}
      executionRuntimeTemplateIndex {value=65535}
      instanceDataHandle {index=65535, source=None, stateHandle {index=65535}}
      instanceTemplateIndex {value=65535}
      outputBindingsBatch {value=65535}
     }
    }
   }]
  transitions [
   {id=#12, trigger=ON_STATE_SUCCEEDED, state {id=#7, name=Loop}}
   {id=#13, trigger=ON_STATE_FAILED, state {id=#14, name=Root, linkType=FAILED}}
  ]
 }
 Loop {
  id=#7
  type=STATE
  children [#15, #16, #17]
  tasks []
  transitions []
 }
 Reload {
  id=#15
  type=STATE
  enter_conditions [
   {
    id=#18
    description=Integer Compare
    properties {
     iD=#19
     instance {_structType=/Script/StateTreeModule.StateTreeCompareIntConditionInstanceData, left=0, right=0}
     node {
      name=Integer Compare
      _structType=/Script/StateTreeModule.StateTreeCompareIntCondition
      bInvert=false
      deltaIndent=0
      evaluationMode=Evaluated
      operand=And
      operator=Equal
      bindingsBatch {value=65535}
      executionRuntimeTemplateIndex {value=65535}
      instanceDataHandle {index=65535, source=None, stateHandle {index=65535}}
      instanceTemplateIndex {value=65535}
      outputBindingsBatch {value=65535}
     }
    }
   }
   {
    id=#20
    description=Bool Compare
    properties {
     iD=#21
     instance {_structType=/Script/StateTreeModule.StateTreeCompareBoolConditionInstanceData, bLeft=false, bRight=true}
     node {
      name=Bool Compare
      _structType=/Script/StateTreeModule.StateTreeCompareBoolCondition
      bInvert=false
      deltaIndent=0
      evaluationMode=Evaluated
      operand=And
      bindingsBatch {value=65535}
      executionRuntimeTemplateIndex {value=65535}
      instanceDataHandle {index=65535, source=None, stateHandle {index=65535}}
      instanceTemplateIndex {value=65535}
      outputBindingsBatch {value=65535}
     }
    }
   }
  ]
  tasks [{
    id=#22
    description=Reload Weapon
    properties {
     iD=#23
     instance {_structType=/Script/FortniteAIServer.FortStateTreeReloadWeaponTaskInstanceData, weapon=null}
     node {
      name=Reload Weapon
      _structType=/Script/FortniteAIServer.FortStateTreeReloadWeaponTask
      bCanEditConsideredForCompletion=true
      bConsideredForCompletion=true
      bHasTaskCompletionDelegateDispatcher=false
      bTaskEnabled=true
      transitionHandlingPriority=Normal
      bindingsBatch {value=65535}
      executionRuntimeTemplateIndex {value=65535}
      instanceDataHandle {index=65535, source=None, stateHandle {index=65535}}
      instanceTemplateIndex {value=65535}
      outputBindingsBatch {value=65535}
     }
    }
   }]
  transitions [{id=#24, trigger=ON_STATE_COMPLETED, state {id=#7, name=Loop}}]
 }
 FireWeapon {
  id=#16
  type=STATE
  enter_conditions [
   {
    id=#25
    description=Bool Compare
    properties {
     iD=#26
     instance {_structType=&StateTreeCompareBoolConditionInstanceData, bLeft=false, bRight=true}
     node {
      name=Bool Compare
      _structType=&StateTreeCompareBoolCondition
      bInvert=false
      deltaIndent=0
      evaluationMode=Evaluated
      operand=And
      bindingsBatch {value=65535}
      executionRuntimeTemplateIndex {value=65535}
      instanceDataHandle {index=65535, source=None, stateHandle {index=65535}}
      instanceTemplateIndex {value=65535}
      outputBindingsBatch {value=65535}
     }
    }
   }
   {
    id=#27
    description=Integer Compare
    properties {
     iD=#28
     instance {_structType=&StateTreeCompareIntConditionInstanceData, left=0, right=0}
     node {
      name=Integer Compare
      _structType=&StateTreeCompareIntCondition
      bInvert=false
      deltaIndent=0
      evaluationMode=Evaluated
      operand=And
      operator=NotEqual
      bindingsBatch {value=65535}
      executionRuntimeTemplateIndex {value=65535}
      instanceDataHandle {index=65535, source=None, stateHandle {index=65535}}
      instanceTemplateIndex {value=65535}
      outputBindingsBatch {value=65535}
     }
    }
   }
  ]
  tasks [
   {
    id=#29
    description=Bot Get Firing Pattern
    properties {
     iD=#30
     instance {
      _structType=/Script/FortniteAIServer.FortStateTreeBotGetFiringPatternTaskInstanceData
      botController=null
      currentWeapon=null
      firingTimes []
     }
     node {
      name=Bot Get Firing Pattern
      _structType=/Script/FortniteAIServer.FortStateTreeBotGetFiringPatternTask
      bCanEditConsideredForCompletion=true
      bConsideredForCompletion=true
      bHasTaskCompletionDelegateDispatcher=false
      bTaskEnabled=true
      transitionHandlingPriority=Normal
      bindingsBatch {value=65535}
      executionRuntimeTemplateIndex {value=65535}
      instanceDataHandle {index=65535, source=None, stateHandle {index=65535}}
      instanceTemplateIndex {value=65535}
      outputBindingsBatch {value=65535}
     }
    }
   }
   {
    id=#31
    description=Bot Fire Weapon
    properties {
     iD=#32
     instance {
      _structType=/Script/FortniteAIServer.FortStateTreeBotFireWeaponTaskInstanceData
      behavior=HoldFireForPattern
      botController=null
      firingTimes []
     }
     node {
      name=Bot Fire Weapon
      _structType=/Script/FortniteAIServer.FortStateTreeBotFireWeaponTask
      bCanEditConsideredForCompletion=true
      bConsideredForCompletion=true
      bHasTaskCompletionDelegateDispatcher=false
      bTaskEnabled=true
      transitionHandlingPriority=Normal
      bindingsBatch {value=65535}
      executionRuntimeTemplateIndex {value=65535}
      instanceDataHandle {index=65535, source=None, stateHandle {index=65535}}
      instanceTemplateIndex {value=65535}
      outputBindingsBatch {value=65535}
     }
    }
   }
   {
    id=#33
    description="Set Weapon Targeting (ADS)"
    properties {
     iD=#34
     instance {_structType=/Script/FortniteAIServer.FortStateTreeSetWeaponTargetingTaskInstanceData, bShouldActivateTargeting=false, weapon=null}
     node {
      name="Set Weapon Targeting (ADS)"
      _structType=/Script/FortniteAIServer.FortStateTreeSetWeaponTargetingTask
      bCanEditConsideredForCompletion=true
      bConsideredForCompletion=true
      bHasTaskCompletionDelegateDispatcher=false
      bTaskEnabled=true
      transitionHandlingPriority=Normal
      bindingsBatch {value=65535}
      executionRuntimeTemplateIndex {value=65535}
      instanceDataHandle {index=65535, source=None, stateHandle {index=65535}}
      instanceTemplateIndex {value=65535}
      outputBindingsBatch {value=65535}
     }
    }
   }
  ]
  transitions [{
    id=#35
    trigger=ON_TICK
    conditions [{
      id=#36
      description=Integer Compare
      properties {
       iD=#37
       instance {_structType=&StateTreeCompareIntConditionInstanceData, left=0, right=0}
       node {
        name=Integer Compare
        _structType=&StateTreeCompareIntCondition
        bInvert=false
        deltaIndent=0
        evaluationMode=Evaluated
        operand=And
        operator=Equal
        bindingsBatch {value=65535}
        executionRuntimeTemplateIndex {value=65535}
        instanceDataHandle {index=65535, source=None, stateHandle {index=65535}}
        instanceTemplateIndex {value=65535}
        outputBindingsBatch {value=65535}
       }
      }
     }]
    state {id=#7, name=Loop}
   }]
 }
 Wait {
  id=#17
  type=STATE
  tasks []
  transitions [{id=#38, trigger=ON_TICK, state {id=#7, name=Loop}}]
 }
]
