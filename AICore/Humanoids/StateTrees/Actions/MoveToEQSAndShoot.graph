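# Humanoid AI action StateTree that moves through EQS-driven positioning while managing weapon fire.
# The Manage Weapon state runs the ManageWeapon building-block tree in parallel, with firing and reloading allowed and aim-down-sights off.
# Its child MoveToEQS linked-asset state re-enters itself on success for continuous repositioning, and fails the tree when the move fails.
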StateTree /AICore/Humanoids/StateTrees/Actions/MoveToEQSAndShoot
states [
 Root {id=#1, type=STATE, children [#2], tasks [], transitions []}
 Manage Weapon {
  id=#2
  type=STATE
  children [#3]
  tasks [{
    id=#4
    description=Run Manage Weapon
    properties {
     iD=#5
     instance {
      _structType=/Script/StateTreeModule.StateTreeRunParallelStateTreeTaskInstanceData
      stateTree {
       stateTree=/AICore/Humanoids/StateTrees/Actions/BuildingBlock/ManageWeapon
       parameters {isAllowedToFire=true, isAllowedToReload=true, shouldADS=false}
       propertyOverrides [#6]
      }
     }
     node {
      name=Run Manage Weapon
      _structType=/Script/StateTreeModule.StateTreeRunParallelStateTreeTask
      bCanEditConsideredForCompletion=true
      bConsideredForCompletion=true
      bHasTaskCompletionDelegateDispatcher=false
      bShouldCopyParametersOnExitState=false
      bShouldCopyParametersOnTick=false
      bTaskEnabled=true
      eventHandlingPriority=Normal
      transitionHandlingPriority=Normal
      bindingsBatch {value=65535}
      executionRuntimeTemplateIndex {value=65535}
      instanceDataHandle {index=65535, source=None, stateHandle {index=65535}}
      instanceTemplateIndex {value=65535}
      outputBindingsBatch {value=65535}
      stateTreeOverrideTag {tagName=None}
     }
    }
   }]
  transitions []
 }
 MoveToEQS {
  id=#3
  type=LINKED_ASSET
  transitions [
   {
    id=#7
    priority=null
    trigger=ON_STATE_SUCCEEDED
    conditions []
    state {id=#3, name=MoveToEQS, linkType=GOTO_STATE}
   }
   {
    id=#8
    priority=null
    trigger=ON_STATE_FAILED
    conditions []
    state {id=#9, name=Root, linkType=FAILED}
   }
  ]
 }
]
