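# StateTree action for humanoid AI that moves to a token position while shooting, built from shared building-block trees.
# A global task gets the target actor, while Root sets AI focus and flags the tactical action.
# Nested child states run ManageWeapon and MoveToTokenPosition as parallel subtrees with parameter overrides, so weapon handling and movement overlap.
# The innermost move state ends the tree as succeeded or failed.
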StateTree /AICore/Humanoids/StateTrees/Actions/MoveToTokenPositionAndShoot
global_tasks [{
  id=#1
  description=Get Target Actor
  properties {
   iD=#2
   instance {_structType=&FortStateTreeGetTargetActorTaskInstanceData}
   node {name=Get Target Actor, _structType=&FortStateTreeGetTargetActorTask}
  }
 }]
states [
 Root {
  id=#3
  type=STATE
  children [#4]
  tasks [
   {
    id=#5
    description="(Fortnite) Focus a specific Actor and/or point/location"
    properties {
     iD=#6
     instance {
      _structType=/Script/FortniteAIServer.FortAthenaFocusAtStateTreeTaskInstanceData
      actor=null
      bSetBackOnExit=true
      bShouldSetFocus=true
      focusActor=null
      focusActorOffset=<0 0 0>
      focusWorldPoint=<0 0 0>
     }
     node {
      name="(Fortnite) Focus a specific Actor and/or point/location"
      _structType=/Script/FortniteAIServer.FortAthenaFocusAtStateTreeTask
      bCanEditConsideredForCompletion=true
      bConsideredForCompletion=true
      bHasTaskCompletionDelegateDispatcher=false
      bTaskEnabled=true
      transitionHandlingPriority=Normal
      bindingsBatch {value=65535}
      executionRuntimeTemplateIndex {value=65535}
      instanceDataHandle {index=65535, source=None, stateHandle {index=65535}}
      instanceTemplateIndex {value=65535}
      outputBindingsBatch {value=65535}
     }
    }
   }
   {
    id=#7
    description=Set In Tactical Action
    properties {
     iD=#8
     instance {
      _structType=/Script/FortniteAIServer.FortStateTreeSetInTacticalActionTaskInstanceData
      bShouldSetInTacticalAction=true
      controller=null
      tacticalActionIdentifierTag {tagName=None}
     }
     node {
      name=None
      _structType=/Script/FortniteAIServer.FortStateTreeSetInTacticalActionTask
      bCanEditConsideredForCompletion=true
      bConsideredForCompletion=true
      bHasTaskCompletionDelegateDispatcher=false
      bTaskEnabled=true
      transitionHandlingPriority=Normal
      bindingsBatch {value=65535}
      executionRuntimeTemplateIndex {value=65535}
      instanceDataHandle {index=65535, source=None, stateHandle {index=65535}}
      instanceTemplateIndex {value=65535}
      outputBindingsBatch {value=65535}
     }
    }
   }
  ]
  transitions []
 }
 ManageWeapon {
  id=#4
  type=STATE
  children [#9]
  tasks [{
    id=#10
    description=Run ManageWeapon
    properties {
     iD=#11
     instance {
      _structType=/Script/StateTreeModule.StateTreeRunParallelStateTreeTaskInstanceData
      stateTree {
       stateTree=/AICore/Humanoids/StateTrees/Actions/BuildingBlock/ManageWeapon
       parameters {isAllowedToFire=true, isAllowedToReload=true, shouldADS=false}
       propertyOverrides [#12]
      }
     }
     node {
      name=Run ManageWeapon
      _structType=/Script/StateTreeModule.StateTreeRunParallelStateTreeTask
      bCanEditConsideredForCompletion=true
      bConsideredForCompletion=true
      bHasTaskCompletionDelegateDispatcher=false
      bShouldCopyParametersOnExitState=false
      bShouldCopyParametersOnTick=false
      bTaskEnabled=true
      eventHandlingPriority=Normal
      transitionHandlingPriority=Normal
      bindingsBatch {value=65535}
      executionRuntimeTemplateIndex {value=65535}
      instanceDataHandle {index=65535, source=None, stateHandle {index=65535}}
      instanceTemplateIndex {value=65535}
      outputBindingsBatch {value=65535}
      stateTreeOverrideTag {tagName=None}
     }
    }
   }]
  transitions []
 }
 MoveToTokenPosition {
  id=#9
  type=STATE
  tasks [{
    id=#13
    description=Run Parallel Tree
    properties {
     iD=#14
     instance {
      _structType=&StateTreeRunParallelStateTreeTaskInstanceData
      stateTree {
       stateTree=/AICore/Humanoids/StateTrees/Actions/BuildingBlock/MoveToTokenPosition
       parameters {
        envQuery=null
        rerunInterval=2
        shouldSprint=false
        reserveTokenPositionQuery {autoDescription="", tokenStreamVersion=0, userDescription="", queryTokenStream [], tagDictionary []}
       }
       propertyOverrides [
        #15
        #16
        #17
        #18
       ]
      }
     }
     node {
      name=Run Parallel Tree
      _structType=&StateTreeRunParallelStateTreeTask
      bCanEditConsideredForCompletion=true
      bConsideredForCompletion=true
      bHasTaskCompletionDelegateDispatcher=false
      bShouldCopyParametersOnExitState=false
      bShouldCopyParametersOnTick=false
      bTaskEnabled=true
      eventHandlingPriority=Normal
      transitionHandlingPriority=Normal
      bindingsBatch {value=65535}
      executionRuntimeTemplateIndex {value=65535}
      instanceDataHandle {index=65535, source=None, stateHandle {index=65535}}
      instanceTemplateIndex {value=65535}
      outputBindingsBatch {value=65535}
      stateTreeOverrideTag {tagName=None}
     }
    }
   }]
  transitions [
   {id=#19, trigger=ON_STATE_SUCCEEDED, state {id=#20, name=None, linkType=SUCCEEDED}}
   {id=#21, trigger=ON_STATE_FAILED, state {id=#20, name=Root, linkType=FAILED}}
  ]
 }
]
