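# StateTree action for Fortnite humanoid AI that uses an item on itself, repeating up to a use limit.
# Root holds a focus task, restored on exit, and an integer variable while a linked asset runs the use.
# On success, Check Uses loops back if Max Uses is negative (unlimited) or Num Uses is below it.
# Otherwise the tree succeeds; a failed use fails it.
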StateTree /AICore/Humanoids/StateTrees/Actions/UseItemOnSelf
states [
 Root {
  id=#1
  type=STATE
  children [#2, #3]
  tasks [
   {
    id=#4
    description="(Fortnite) Focus a specific Actor and/or point/location"
    properties {
     iD=#5
     instance {
      _structType=/Script/FortniteAIServer.FortAthenaFocusAtStateTreeTaskInstanceData
      actor=null
      bSetBackOnExit=true
      bShouldSetFocus=true
      focusActor=null
      focusActorOffset=<100 0 0>
      focusWorldPoint=<0 0 0>
     }
     node {
      name=None
      _structType=/Script/FortniteAIServer.FortAthenaFocusAtStateTreeTask
      bCanEditConsideredForCompletion=true
      bConsideredForCompletion=true
      bHasTaskCompletionDelegateDispatcher=false
      bTaskEnabled=true
      transitionHandlingPriority=Normal
      bindingsBatch {value=65535}
      executionRuntimeTemplateIndex {value=65535}
      instanceDataHandle {index=65535, source=None, stateHandle {index=65535}}
      instanceTemplateIndex {value=65535}
      outputBindingsBatch {value=65535}
     }
    }
   }
   {
    id=#6
    description="(Fortnite) Arithmetic: Make Int"
    properties {
     iD=#7
     instance {_structType=/Script/FortniteAIServer.FortAthenaMakeIntVariableStateTreeTaskInstanceData, defaultIntValue=0, outInt {int=0}}
     node {
      name=None
      _structType=/Script/FortniteAIServer.FortAthenaMakeIntVariableStateTreeTask
      bCanEditConsideredForCompletion=true
      bConsideredForCompletion=true
      bHasTaskCompletionDelegateDispatcher=false
      bResetOnReselect=false
      bTaskEnabled=true
      transitionHandlingPriority=Normal
      bindingsBatch {value=65535}
      executionRuntimeTemplateIndex {value=65535}
      instanceDataHandle {index=65535, source=None, stateHandle {index=65535}}
      instanceTemplateIndex {value=65535}
      outputBindingsBatch {value=65535}
     }
    }
   }
  ]
  transitions []
 }
 UseItem {
  id=#2
  type=LINKED_ASSET
  transitions [
   {
    id=#8
    priority=null
    trigger=ON_STATE_SUCCEEDED
    conditions []
    state {id=#3, name=Check Uses, linkType=GOTO_STATE}
   }
   {
    id=#9
    priority=null
    trigger=ON_STATE_FAILED
    conditions []
    state {id=#10, name=Root, linkType=FAILED}
   }
  ]
 }
 Check Uses {
  id=#3
  type=STATE
  tasks [{
    id=#11
    description="(Fortnite) Arithmetic: Make Int"
    properties {
     iD=#12
     instance {_structType=&FortAthenaMakeIntVariableStateTreeTaskInstanceData, defaultIntValue=0, outInt {int=0}}
     node {
      name=None
      _structType=&FortAthenaMakeIntVariableStateTreeTask
      bCanEditConsideredForCompletion=true
      bConsideredForCompletion=true
      bHasTaskCompletionDelegateDispatcher=false
      bResetOnReselect=false
      bTaskEnabled=true
      transitionHandlingPriority=Normal
      bindingsBatch {value=65535}
      executionRuntimeTemplateIndex {value=65535}
      instanceDataHandle {index=65535, source=None, stateHandle {index=65535}}
      instanceTemplateIndex {value=65535}
      outputBindingsBatch {value=65535}
     }
    }
   }]
  transitions [
   {
    id=#13
    trigger=ON_TICK
    conditions [
     {
      id=#14
      description="Parameters.Max Uses < 0"
      properties {
       iD=#15
       instance {_structType=/Script/StateTreeModule.StateTreeCompareIntConditionInstanceData, left=0, right=0}
       node {
        name=None
        _structType=/Script/StateTreeModule.StateTreeCompareIntCondition
        bInvert=false
        deltaIndent=0
        evaluationMode=Evaluated
        operand=And
        operator=Less
        bindingsBatch {value=65535}
        executionRuntimeTemplateIndex {value=65535}
        instanceDataHandle {index=65535, source=None, stateHandle {index=65535}}
        instanceTemplateIndex {value=65535}
        outputBindingsBatch {value=65535}
       }
      }
     }
     {
      id=#16
      description="Parameters.Num Uses < Parameters.Max Uses"
      properties {
       expressionOperand=Or
       iD=#17
       instance {_structType=&StateTreeCompareIntConditionInstanceData, left=0, right=0}
       node {
        name=None
        _structType=&StateTreeCompareIntCondition
        bInvert=false
        deltaIndent=0
        evaluationMode=Evaluated
        operand=And
        operator=Less
        bindingsBatch {value=65535}
        executionRuntimeTemplateIndex {value=65535}
        instanceDataHandle {index=65535, source=None, stateHandle {index=65535}}
        instanceTemplateIndex {value=65535}
        outputBindingsBatch {value=65535}
       }
      }
     }
    ]
    state {id=#2, name=UseItem}
   }
   {id=#18, trigger=ON_TICK, state {id=#10, name=Root, linkType=SUCCEEDED}}
  ]
 }
]
