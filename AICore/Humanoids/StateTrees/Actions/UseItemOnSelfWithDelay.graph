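# AI humanoid action StateTree that, by its name, has the bot use an item on itself with a delay.
# Root runs a Fortnite focus task with a forward actor offset, restoring focus on exit, while its child state runs.
# The child, UseItemWithDelay, is a linked-asset state that re-enters itself on success, repeating the use, and fails the tree on failure.
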StateTree /AICore/Humanoids/StateTrees/Actions/UseItemOnSelfWithDelay
states [
 Root {
  id=#1
  type=STATE
  children [#2]
  tasks [{
    id=#3
    description="(Fortnite) Focus a specific Actor and/or point/location"
    properties {
     iD=#4
     instance {
      _structType=/Script/FortniteAIServer.FortAthenaFocusAtStateTreeTaskInstanceData
      actor=null
      bSetBackOnExit=true
      bShouldSetFocus=true
      focusActor=null
      focusActorOffset=<100 0 0>
      focusWorldPoint=<0 0 0>
     }
     node {
      name=None
      _structType=/Script/FortniteAIServer.FortAthenaFocusAtStateTreeTask
      bCanEditConsideredForCompletion=true
      bConsideredForCompletion=true
      bHasTaskCompletionDelegateDispatcher=false
      bTaskEnabled=true
      transitionHandlingPriority=Normal
      bindingsBatch {value=65535}
      executionRuntimeTemplateIndex {value=65535}
      instanceDataHandle {index=65535, source=None, stateHandle {index=65535}}
      instanceTemplateIndex {value=65535}
      outputBindingsBatch {value=65535}
     }
    }
   }]
  transitions []
 }
 UseItemWithDelay {
  id=#2
  type=LINKED_ASSET
  transitions [
   {
    id=#5
    priority=null
    trigger=ON_STATE_SUCCEEDED
    conditions []
    state {id=#2, name=UseItemWithDelay, linkType=GOTO_STATE}
   }
   {
    id=#6
    priority=null
    trigger=ON_STATE_FAILED
    conditions []
    state {id=#7, name=Root, linkType=FAILED}
   }
  ]
 }
]
